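# Niagara effect for the ProfessorPup character cosmetic, emitting an exhaust ribbon, flipbook smoke and flickering sparkles from two skeletal-mesh sockets.
# Each emitter splits particles into two spawn groups whose index selects the socket on User.SkeletalMesh.
# Particles re-sample that socket every update, adding an offset rotated by socket orientation and scaled by normalized age, so trails follow the character without velocity.
# User.IsSprinting raises the smoke spawn rate, while curves animate color, size and material parameters.
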NiagaraSystem /BRCosmetics/Effects/Fort_Effects/Effects/Characters/Athena_Parts/ProfessorPup/NS_ProfessorPup_Good
dependencies {
 usedDataInterfaces [&NiagaraDataInterfaceColorCurve, &NiagaraDataInterfaceCurve, &NiagaraDataInterfaceVector2DCurve]
 usedDynamicInputs []
 usedModules [&SystemState]
}
emitters {
 ExhaustRibbon {
  emitterUpdateScript [
   &EmitterState()
   &SpawnRate(SpawnRate=30, Use Spawn Probability=true)
  ]
  particleSpawnScript [
   &FN_SetSpawnGroup_v2(Number of Spawn Groups=2)
   &InitializeRibbon(
    Color=<rgba 1 1 1 0.6>
    Lifetime=0.9
    Module.Write Color=true
    Module.Write Lifetime=true
    Module.Write Mass=false
    Module.Write Material Random=true
    Module.Write Position=false
    Module.Write RibbonTwist=false
    Module.Write RibbonWidth=true
    Ribbon Width=25
   )
  ]
  particleUpdateScript [
   &ParticleState()
   /SetVariables(Particles.RangeSmokePosition:Vector3f=&MakeVector(
     X=0
     Y=&FloatFromCurve(FloatCurve=DataInterface() {propertyValues {Curve {keys [<at 0 0>,<at 1 0>]}}})
     Z=45
    )) {isSetParameters=true, module=SetVariables_#1}
   &SampleSkeletalMeshSkeleton(
    Interpolate Bone Sampling=false
    Mesh:NiagaraDataInterfaceSkeletalMesh=$User.SkeletalMesh
    Sampled Rotation Coordinate Space:ENiagaraCoordinateSpace=::World
    Spawn Mode:ENiagaraSkeletalSpawning=::NewEnumerator1
    Specific Socket Index:NiagaraInt32=&SpawnGroup()
   )
   &SkeletalMeshLocation(
    Bone / Socket Index=&SpawnGroup()
    Inherit Velocity=false
    Interpolate Mesh Sampling=false
    MaskSpawnGroup=false
    Mesh Sampling Type=::NewEnumerator5
    Sampled Position Offset=&RotateVectorByQuaternion(
     Vector To Rotate=&Multiply_VectorByFloat(Vector=$Particles.RangeSmokePosition, Float=$Particles.NormalizedAge)
     Quaternion=$Particles.SampleSkeletalMeshSkeleton.SampledRotation
    )
    Skeletal Mesh=$User.SkeletalMesh
    Socket Sampling Mode=::NewEnumerator5
   )
   &DynamicMaterialParameters(
    Index 0 Param 1=&FloatFromCurve(FloatCurve=DataInterface() {propertyValues {Curve {keys [<at 0 1>,<at 1 0.75>]}}})
    Index 0 Param 3=-6
    Index 0 Param 4=&FloatFromCurve(FloatCurve=DataInterface() {propertyValues {Curve {keys [<at 0 1>,<at 1 0>]}}})
    Module.Param0WriteEnabled=true
    Module.Param1WriteEnabled=true
    Module.Param2WriteEnabled=true
    Module.Param3WriteEnabled=true
   )
   /SetVariables(Particles.RibbonWidth=&Multiply_Float(
     A=$Particles.Initial.RibbonWidth
     B=&FloatFromCurve(FloatCurve=DataInterface() {propertyValues {Curve {keys [<at 0.1 0.25 RCIM_Cubic>, {interpMode=RCIM_Cubic, leaveTangent=1.2519, time=0.9, value=0.7}]}}})
    )) {isSetParameters=true}
   &ScaleColor(
    Scale Alpha=&FloatFromCurve(FloatCurve=DataInterface() {propertyValues {Curve {keys [
         <at 0 0 RCIM_Cubic>
         {arriveTangent=-1.1482, interpMode=RCIM_Cubic, leaveTangent=-1.1482, tangentMode=RCTM_User, time=0.05, value=1}
         <at 1 0>
        ]}}})
    ScaleA=true
    ScaleRGB=true
   )
  ]
  rendererClasses [&NiagaraRibbonRendererProperties]
  renderers [{rendererClass=&NiagaraRibbonRendererProperties}]
 }
 Smoke {
  emitterUpdateScript [
   &EmitterState()
   &SpawnRate(
    SpawnRate=&MakeCustomFloatFromBool(Bool=$User.IsSprinting, True Float=15, False Float=10)
    Use Spawn Probability=false
   )
  ]
  particleSpawnScript [
   &FN_SetSpawnGroup_v2(Number of Spawn Groups=2)
   /Niagara/Modules/Spawn/Initialization/InitializeParticle(
    Color=&UniformRangedLinearColor(Minimum=<rgba 0 0 1 0.4>, Maximum=<rgba 1 0 0 0.5>)
    Lifetime=&UniformRangedFloat(Minimum=0.75, Maximum=1)
    Module.Write Color=true
    Module.Write Lifetime=true
    Module.Write Mass=false
    Module.Write Material Random=true
    Module.Write Position=false
    Module.Write Scale=false
    Module.Write SpriteRotation=true
    Module.Write SpriteSize=true
    Sprite Rotation=&UniformRangedFloat(Minimum=-360, Maximum=360)
    Sprite Size=&Vector2DFromFloat(Value=&UniformRangedFloat(Minimum=10, Maximum=12))
   )
   /SetVariables(Particles.RangeSmokePosition:Vector3f=&UniformRangedVector(Minimum=<0 0 10>, Maximum=<0 0 20>)) {isSetParameters=true, module=SetVariables_#1}
   &DynamicMaterialParameters(
    Index 0 Param 1=&UniformRangedFloat(Minimum=0, Maximum=1)
    Module.Param0WriteEnabled=true
    Module.Param1WriteEnabled=true
    Module.Param2WriteEnabled=true
    Module.Param3WriteEnabled=true
   ) {module=DynamicMaterialParameters001}
  ]
  particleUpdateScript [
   &ParticleState()
   &ScaleColor(
    Scale Alpha=&FloatFromCurve(FloatCurve=DataInterface() {propertyValues {Curve {keys [
         <at 0 0 RCIM_Cubic>
         <at 0.05 1 RCIM_Cubic>
         {arriveTangent=-3.5059, interpMode=RCIM_Cubic, leaveTangent=-3.5059, tangentMode=RCTM_User, time=1, value=0}
        ]}}})
    Scale RGB=&MakeVectorFromLinearColorRGB(Color=&ColorFromCurve(ColorCurve=DataInterface() {propertyValues {
        AlphaCurve {keys [<at 0 1 RCIM_Cubic>]}
        BlueCurve {keys [<at 0 0 RCIM_Cubic>,<at 0.1 0>]}
        GreenCurve {keys [<at 0 0 RCIM_Cubic>,<at 0.1 0>]}
        RedCurve {keys [<at 0 5 RCIM_Cubic>,<at 0.1 3>]}
       }}))
    ScaleA=true
    ScaleRGB=true
   )
   &SampleSkeletalMeshSkeleton(
    Interpolate Bone Sampling=false
    Mesh:NiagaraDataInterfaceSkeletalMesh=$User.SkeletalMesh
    Sampled Rotation Coordinate Space:ENiagaraCoordinateSpace=::Local
    Spawn Mode:ENiagaraSkeletalSpawning=::NewEnumerator1
    Specific Socket Index:NiagaraInt32=&SpawnGroup()
   )
   &SkeletalMeshLocation(
    Bone / Socket Index=&SpawnGroup()
    Inherit Velocity=false
    Interpolate Mesh Sampling=false
    MaskSpawnGroup=false
    Mesh Sampling Type=::NewEnumerator5
    Sampled Position Offset=&RotateVectorByQuaternion(
     Vector To Rotate=&Multiply_VectorByFloat(Vector=$Particles.RangeSmokePosition, Float=$Particles.NormalizedAge)
     Quaternion=$Particles.SampleSkeletalMeshSkeleton.SampledRotation
    )
    Sampled Position Offset Coordinate Space:ENiagaraMeshTransforms=::NewEnumerator1
    Skeletal Mesh=$User.SkeletalMesh
    Socket Sampling Mode=::NewEnumerator5
   )
   &ScaleSpriteSize(Scale Factor=&Vector2DFromFloat(Value=&FloatFromCurve(FloatCurve=DataInterface() {propertyValues {Curve {keys [<at 0 0.5 RCIM_Cubic>,<at 1 2>]}}})))
   /SetVariables(Particles.DetailSmokeSize:Vector2f=&Multiply_Vector2(
     A=$Particles.Initial.DetailSmokeSize
     B=&Vector2DFromFloat(Value=&FloatFromCurve(FloatCurve=DataInterface() {propertyValues {Curve {keys [<at 0 0.5 RCIM_Cubic>,<at 1 2>]}}}))
    )) {isSetParameters=true}
   /SetVariables(Particles.SubUVImageIndex:NiagaraFloat=&FloatFromCurve(FloatCurve=DataInterface() {propertyValues {Curve {keys [{arriveTangent=72.2856, interpMode=RCIM_Cubic, leaveTangent=72.2855, tangentMode=RCTM_User, time=0, value=10}, <at 1 55>]}}})) {isSetParameters=true}
  ]
  rendererClasses [&NiagaraSpriteRendererProperties]
  renderers [{rendererClass=&NiagaraSpriteRendererProperties}]
 }
 Sparkles {
  emitterUpdateScript [
   &EmitterState()
   &SpawnRate(SpawnRate=6, Use Spawn Probability=false)
  ]
  particleSpawnScript [
   &FN_SetSpawnGroup_v2(Number of Spawn Groups=2)
   &InitializeParticle(
    Lifetime=&UniformRangedFloat(Minimum=0.6, Maximum=0.85)
    Module.Write Color=true
    Module.Write Lifetime=true
    Module.Write Mass=false
    Module.Write Material Random=true
    Module.Write Position=false
    Module.Write Scale=false
    Module.Write SpriteRotation=false
    Module.Write SpriteSize=true
    Sprite Size=&Vector2DFromFloat(Value=&UniformRangedFloat(Minimum=3, Maximum=4))
   )
   /Niagara/Modules/Spawn/Location/SphereLocation(
    Coordinate Space:ENiagaraCoordinateSpace=::Simulation
    MaskSpawnGroup=false
    Module.Override Randomness=true
    Module.Override Seed=true
    Randomness Mode:ENiagaraRandomnessMode=::NewEnumerator1
    Sphere Origin:Vector3f=$Particles.SampleSkeletalMeshSkeleton.SampledPosition
    Sphere Radius=1
    Surface Only=false
   ) {enabled=false}
   /SetVariables(Particles.RangeSmokePosition:Vector3f=<0 0 8>) {isSetParameters=true, module=SetVariables_#1}
   /SetVariables(Particles.SphereLocationOffset:Vector3f=&Multiply_VectorByFloat(Vector=$Output.SphereLocation.SphereNormal, Float=8)) {enabled=false, isSetParameters=true}
  ]
  particleUpdateScript [
   &ParticleState()
   &ScaleColor(
    Scale Alpha=&FloatFromCurve(FloatCurve=DataInterface() {propertyValues {Curve {keys [
         <at 0 0 RCIM_Cubic>
         {arriveTangent=0.1538, interpMode=RCIM_Cubic, leaveTangent=0.1538, time=0.3, value=1}
         {arriveTangent=-0.119, interpMode=RCIM_Cubic, leaveTangent=-0.119, time=0.52, value=0.08}
         {arriveTangent=0.5, interpMode=RCIM_Cubic, leaveTangent=0.5, time=0.72, value=0.95}
         {arriveTangent=0.2143, interpMode=RCIM_Cubic, leaveTangent=0.2143, time=0.8, value=0.22}
         {arriveTangent=4.0526, interpMode=RCIM_Cubic, leaveTangent=4.0526, time=0.86, value=0.98}
         <at 0.99 0.99 RCIM_Cubic>
        ]}}})
    Scale RGB=&MakeVectorFromLinearColorRGB(Color=&ColorFromCurve(ColorCurve=DataInterface() {propertyValues {
        AlphaCurve {keys []}
        BlueCurve {keys [<at 0.15 0.4936 RCIM_Cubic>,<at 1 1.49 RCIM_Cubic>]}
        GreenCurve {keys [<at 0.15 36.2246 RCIM_Cubic>,<at 1 1.7567 RCIM_Cubic>]}
        RedCurve {keys [<at 0.15 60 RCIM_Cubic>,<at 1 2 RCIM_Cubic>]}
       }}))
    ScaleA=true
    ScaleRGB=true
   )
   &SampleSkeletalMeshSkeleton(
    Interpolate Bone Sampling=false
    Mesh:NiagaraDataInterfaceSkeletalMesh=$User.SkeletalMesh
    Spawn Mode:ENiagaraSkeletalSpawning=::NewEnumerator1
    Specific Socket Index:NiagaraInt32=&SpawnGroup()
   )
   &SkeletalMeshLocation(
    Bone / Socket Index=&SpawnGroup()
    Inherit Velocity=false
    Interpolate Mesh Sampling=false
    MaskSpawnGroup=false
    Mesh Sampling Type=::NewEnumerator5
    Sampled Position Offset=&RotateVectorByQuaternion(
     Vector To Rotate=&Add_Vector(
      A=<0 0 4>
      B=&Multiply_VectorByFloat(
       Vector=&Add_Vector(A=$Particles.SphereLocationOffset, B=$Particles.RangeSmokePosition)
       Float=$Particles.NormalizedAge
      )
     )
     Quaternion=$Particles.SampleSkeletalMeshSkeleton.SampledRotation
    )
    Skeletal Mesh=$User.SkeletalMesh
    Socket Sampling Mode=::NewEnumerator5
   )
   &ScaleSpriteSize(Scale Factor=&Vector2DFromCurve(
     Vector2Curve=DataInterface() {propertyValues {
       XCurve {keys [
         {arriveTangent=50.4225, interpMode=RCIM_Cubic, leaveTangent=50.4232, tangentMode=RCTM_User, time=0, value=0.5}
         {arriveTangent=2, interpMode=RCIM_Cubic, leaveTangent=2, time=0.15, value=0.28}
         {arriveTangent=2.15, interpMode=RCIM_Cubic, leaveTangent=2.15, time=0.25, value=1}
         {arriveTangent=-0.1053, interpMode=RCIM_Cubic, leaveTangent=-0.1053, time=0.35, value=0.71}
         {arriveTangent=1.0357, interpMode=RCIM_Cubic, leaveTangent=1.0357, time=0.44, value=0.98}
         {arriveTangent=-2.0652, interpMode=RCIM_Cubic, leaveTangent=-2.0652, time=0.63, value=1}
         <at 0.9 0.03 RCIM_Cubic>
        ]}
       YCurve {keys [
         <at 0 2.8 RCIM_Cubic>
         {arriveTangent=-7.2, interpMode=RCIM_Cubic, leaveTangent=-7.2, time=0.15, value=1}
         <at 0.25 1>
         {arriveTangent=-5.6, interpMode=RCIM_Cubic, leaveTangent=-5.6, time=0.2615, value=1}
         {arriveTangent=-0.056, interpMode=RCIM_Cubic, leaveTangent=-0.056, time=0.35, value=0.44}
         {arriveTangent=2, interpMode=RCIM_Cubic, leaveTangent=2, time=0.44, value=0.99}
         {arriveTangent=-2.087, interpMode=RCIM_Cubic, leaveTangent=-2.087, time=0.63, value=1}
         <at 0.9 0.03 RCIM_Cubic>
        ]}
      }}
     ScaleCurve=<1 1>
    ))
   &DynamicMaterialParameters(
    Index 0 Param 1=&FloatFromCurve(FloatCurve=DataInterface() {propertyValues {Curve {keys [<at 0.3 1 RCIM_Cubic>,<at 1 -0.1>]}}})
    Index 0 Param 2=&FloatFromCurve(FloatCurve=DataInterface() {propertyValues {Curve {keys [<at 0 1 RCIM_Cubic>,<at 0.8 8.8 RCIM_Cubic>]}}})
    Index 0 Param 4=&FloatFromCurve(FloatCurve=DataInterface() {propertyValues {Curve {keys [<at 0.3 0>,<at 0.5 0.5>]}}})
    Module.Param0WriteEnabled=true
    Module.Param1WriteEnabled=true
    Module.Param2WriteEnabled=true
    Module.Param3WriteEnabled=true
   )
  ]
  rendererClasses [&NiagaraSpriteRendererProperties]
  renderers [{rendererClass=&NiagaraSpriteRendererProperties}]
 }
}
summary {
 systemName=NS_ProfessorPup_Good
 emitters [
  {emitterName=Smoke, rendererClasses [&NiagaraSpriteRendererProperties]}
  {emitterName=ExhaustRibbon, rendererClasses [&NiagaraRibbonRendererProperties]}
  {emitterName=Sparkles, rendererClasses [&NiagaraSpriteRendererProperties]}
 ]
 userVariables [
  User.bIsFrontEndPreview {type=NiagaraBool}
  User.IsSprinting {type=NiagaraBool}
  User.SkeletalMesh {
   type=NiagaraDataInterfaceSkeletalMesh
   defaultValue {
    struct=NiagaraExt_VariableValue_DataInterface
    value {dataInterface=&self:NiagaraDataInterfaceSkeletalMesh_0}
   }
  }
 ]
}
